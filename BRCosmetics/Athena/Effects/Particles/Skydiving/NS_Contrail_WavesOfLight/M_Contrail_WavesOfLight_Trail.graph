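# Translucent, unlit, two-sided trail material for Niagara mesh particles in the WavesOfLight skydiving contrail.
# A scrolling flow texture and soft smoke mask build the emissive color, tinted by particle color or a height gradient, then adjusted by time of day.
# Opacity combines masks, inverted fresnel, camera and depth fades.
# Dynamic parameters randomize flow and drive world position offset: sine waves along normals, plus optional trailing bend.
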Material /BRCosmetics/Athena/Effects/Particles/Skydiving/NS_Contrail_WavesOfLight/M_Contrail_WavesOfLight_Trail
graph {
 properties {
  bCastRayTracedShadows=false
  bUseTranslucencyVertexFog=false
  bUsedWithNiagaraMeshParticles=true
  blendMode=BLEND_Translucent
  pixelDepthOffsetMode=PDOM_Legacy
  shadingModel=MSM_Unlit
  twoSided=true
  usedShadingModels=Unlit
 }
 nodes [
  n1 = MaterialFunctionCall(Float2=TextureCoordinate()) {materialFunction=&BreakOutFloat2Components}
  n2 = DynamicParameter() {defaultValue=<rgba 0 0 1 0>, paramNames [UVTilingRandom, TimeOffsetRandom, WPOStrength, TrailingModeON]}
  n3 = VectorParameter() {group=Mask, parameterName=Mask UV, defaultValue=<rgba 0.2 1 0.5 0>}
  n4 = Add(A=@n2.UVTilingRandom, B=@n3.R)
  n5 = Time()
  n6 = Add(
   A=Multiply(
    A=MaterialFunctionCall(X=@n1.G, Y=@n1.G) {materialFunction=&MakeFloat2}.Result
    B=MaterialFunctionCall(X=@n4, Y=@n3.G) {materialFunction=&MakeFloat2}.Result
   )
   B=Multiply(A=MaterialFunctionCall(X=@n3.B, Y=@n3.B) {materialFunction=&MakeFloat2}.Result, B=@n5)
  )
  n7 = TextureCoordinate()
  n8 = Sine(None=Multiply(A=ComponentMask(None=@n7) {r=true}, B=Constant() {r=3}))
  n9 = ComponentMask(None=@n7) {g=true}
  n10 = Constant() {r=0.75}
  n11 = Sine(None=Add(A=@n5, B=Multiply(A=@n9, B=@n10)))
  Add(A=@n6, B=Multiply(A=Multiply(A=@n8, B=@n11), B=Constant2Vector() {g=0.06}))
  Constant()
  n12 = VertexNormalWS()
  n13 = DynamicParameter() {defaultValue=<rgba 0 0 1 0>, paramNames [UVTilingRandom, TimeOffsetRandom, WPOStrength, TrailingModeON]}
  n14 = Add(A=Time(), B=@n13.TimeOffsetRandom)
  n15 = VectorParameter() {group=WPO, parameterName=WPO Sine, channelNames {a="", b=Sine Strength, g=Sine Tile, r=Sine Speed}, defaultValue=<rgba 2 2 1 1>}
  n16 = DynamicParameter() {parameterIndex=2, defaultValue=<rgba 0 0 1 0>, paramNames [Sine Speed, Sine Tile, Param3, Param4]}
  n17 = Add(A=@n15.Sine Speed, B=@n16.Sine Speed)
  n18 = Add(A=@n13.UVTilingRandom, B=ComponentMask(None=TextureCoordinate()) {g=true})
  n19 = Add(A=@n15.Sine Tile, B=@n16.Sine Tile)
  n20 = Sine(None=Add(A=Multiply(A=@n14, B=@n17), B=Multiply(A=@n18, B=@n19) {constB=2}))
  n21 = TextureCoordinate()
  n22 = LinearInterpolate(
   A=Constant3Vector()
   B=Multiply(
    A=MaterialFunctionCall(X=@n20, Y=@n20, Z=@n20) {materialFunction=&MakeFloat3}.Result
    B=@n15.Sine Strength
   ) {constB=0.2}
   Alpha=OneMinus(None=ComponentMask(None=@n21) {g=true})
  )
  n23 = DynamicParameter() {defaultValue=<rgba 0 0 1 0>, paramNames [UVTilingRandom, TimeOffsetRandom, WPOStrength, TrailingModeON]}
  n24 = Add(A=Multiply(A=Multiply(A=@n12, B=@n22), B=@n23.WPOStrength) {constB=10})
  n25 = DynamicParameter() {desc=Parameters set by Niagara, parameterIndex=1, paramNames [X,Y,Z,WPOIntensity]}
  n26 = VectorParameter() {group=WPO, parameterName=WPO Control, channelNames {a=Speed, b=BendPower, g=Max, r=Min}, defaultValue=<rgba -50 50 3 3>}
  n27 = Clamp(
   None=Subtract(
    A=MaterialFunctionCall(X=@n25.X, Y=@n25.X, Z=@n25.X) {materialFunction=&MakeFloat3}.Result
    B=ObjectPositionWS()
   )
   Min=@n26.Min
   Max=@n26.Min
  ) {maxDefault=50, minDefault=-50}
  n28 = ComponentMask(None=TextureCoordinate()) {g=true}
  n29 = Multiply(A=Power(Base=OneMinus(None=@n28), Exp=@n26.BendPower), B=@n25.WPOIntensity)
  LinearInterpolate(A=@n24, B=Add(A=LinearInterpolate(B=@n27, Alpha=@n29), B=@n24), Alpha=@n23.TrailingModeON)
  n30 = Add(A=ComponentMask(None=TextureCoordinate()) {g=true}, B=ScalarParameter() {parameterName=Add})
  n31 = ScalarParameter() {defaultValue=1, parameterName=ColorExponent}
  n32 = Multiply(
   A=LinearInterpolate(
    A=VectorParameter() {parameterName=ColorZero, defaultValue=<rgba 0.7225 0 1 0>}.RGB
    B=VectorParameter() {parameterName=ColorOne, defaultValue=<rgba 0 0.3803 1 0>}.RGB
    Alpha=Power(Base=@n30, Exp=@n31)
   )
   B=ScalarParameter() {defaultValue=1, parameterName=ColorIntensity}
  )
  n33 = TextureCoordinate()
  n34 = VectorParameter() {group=Flow, parameterName=Flow UV, defaultValue=<rgba 1 1 0 3>}
  n35 = DynamicParameter() {defaultValue=<rgba 0 0 1 0>, paramNames [UVTilingRandom, TimeOffsetRandom, WPOStrength, TrailingModeON]}
  n36 = MaterialFunctionCall(X=@n34.R, Y=Add(A=@n35.UVTilingRandom, B=@n34.G)) {materialFunction=&MakeFloat2}
  n37 = MaterialFunctionCall(X=@n34.B, Y=@n34.B) {materialFunction=&MakeFloat2}
  n38 = Add(A=@n35.TimeOffsetRandom, B=Time())
  n39 = TextureSampleParameter2D(UVs=Add(A=Multiply(A=@n33, B=@n36.Result), B=Multiply(A=@n37.Result, B=@n38))) {
   group=Flow
   parameterName=Flow Texture
   samplerType=SAMPLERTYPE_LinearColor
   texture=/BRCosmetics/Athena/Effects/Particles/Skydiving/NS_Contrail_Turtleneck_Crystal/T_WavePattern_Chromatic
  }
  n40 = VectorParameter() {group=Flow, parameterName=Flow Refine, defaultValue=<rgba 1 1 1 0.5>}
  n41 = LinearInterpolate(A=Multiply(A=Power(Base=@n39.A, Exp=@n40.R), B=@n40.G), B=Multiply(A=@n39.A, B=@n40.B), Alpha=@n40.A)
  n42 = VectorParameter() {group=Flow, parameterName=Flow Color, channelNames {a=Add, b="", g="", r=""}, defaultValue=<rgba 1 1 1 0>}
  n43 = Add(A=Desaturation(None=@n39.RGB, Fraction=ScalarParameter() {defaultValue=-10, group=Flow, parameterName=Flow Desaturation}), B=@n42.Add)
  n44 = ShadingPathSwitch(Default=@n41, Deferred=@n41, Forward=@n43, Mobile=@n43)
  n45 = VectorParameter() {group=Flow, parameterName=Inside Color, channelNames {a=SizePower, b="", g="", r=""}, defaultValue=<rgba 1 1 1 1>}
  n46 = TextureCoordinate()
  n47 = TextureSampleParameter2D(UVs=@n6) {
   group=Mask
   parameterName=Mask Texture
   samplerType=SAMPLERTYPE_Grayscale
   texture=/Game/Effects/Fort_Effects/Textures/Smoke/T_Smoke_Trail_Soft_NoClamp
  }
  n48 = Multiply(A=Saturate(None=ComponentMask(None=@n46) {g=true}), B=@n47.R)
  n49 = LinearInterpolate(A=Multiply(A=Multiply(A=@n44, B=@n43), B=@n42), B=@n45, Alpha=Saturate(None=Power(Base=@n48, Exp=@n45.SizePower)))
  n50 = ParticleColor()
  n51 = StaticSwitchParameter(True=Desaturation(None=Multiply(A=@n32, B=@n49), Fraction=ScalarParameter() {parameterName=Saturation}), False=Multiply(A=@n49, B=@n50.RGB)) {parameterName=UseHeightLerp}
  n52 = VectorParameter() {parameterName=ColorTimeOfDay, channelNames {a=Night, b=Evening, g=Day, r=Morning}, defaultValue=<rgba 1 1 1 1>}
  n53 = MaterialFunctionCall(
   Color_Morning=Multiply(A=@n51, B=@n52.Morning)
   Color_Day=Multiply(A=@n51, B=@n52.Day)
   Color_Evening=Multiply(A=@n51, B=@n52.Evening)
   Color_Night=Multiply(A=@n51, B=@n52.Night)
  ) {materialFunction=&Color_Time_of_Day}
  n54 = ScalarParameter() {defaultValue=5, parameterName=Mobile Multiplier}
  MaterialFunctionCall(Emissive=ShadingPathSwitch(Default=@n53.Result, Deferred=@n53.Result, Mobile=Multiply(A=@n54, B=@n53.Result)) {desc=Migrated from FeatureLevelSwitch}) {materialFunction=&MF_FX_EmissiveAdjustments}
  n55 = MaterialFunctionCall(Distance=Constant() {r=128}) {materialFunction=&Near_Camera_Fade}
  n56 = Fresnel() {exponent=10}
  n57 = Multiply(
   A=TextureSample(UVs=TextureCoordinate()) {samplerSource=SSM_Clamp_WorldGroupSettings, texture=/Game/Effects/Fort_Effects/Textures/Masks/T_Square_Gradient_M}.R
   B=@n48
  )
  n58 = VectorParameter() {parameterName=Base, channelNames {a=NoiseStrength, b=MaskMultiply, g=DepthFade, r=Refraction}, defaultValue=<rgba 0 50 1 0.2>}
  n59 = Multiply(A=Multiply(A=Multiply(A=@n50.A, B=@n55.Result), B=OneMinus(None=@n56)), B=Saturate(None=Multiply(A=@n57, B=@n58.MaskMultiply)))
  n60 = DepthFade(Opacity=@n59, FadeDistance=@n58.DepthFade)
  Multiply(A=ShadingPathSwitch(Default=@n60, Deferred=@n60, Forward=@n59, Mobile=@n59), B=ScalarParameter() {defaultValue=1, parameterName=Alpha})
 ]
}
