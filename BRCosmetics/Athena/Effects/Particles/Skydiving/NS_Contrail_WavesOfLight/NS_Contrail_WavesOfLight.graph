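# Waves of Light skydiving contrail cosmetic, a cyan Niagara effect attached to bones of the player's skeletal mesh.
# Mesh trails on five bone groups align to skydive velocity, joined by color-pulsing sprites, two-group flares and lights, and streaming and surrounding sparkles.
# Dive amount drives trail shape, sparkle offset, pulse rate and light brightness, while time of day scales sprite and light intensity.
# Front-end modifiers adjust trail scale and surrounding sparkle offset.
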NiagaraSystem /BRCosmetics/Athena/Effects/Particles/Skydiving/NS_Contrail_WavesOfLight/NS_Contrail_WavesOfLight
dependencies {
 usedDataInterfaces [&NiagaraDataInterfaceArrayFloat, &NiagaraDataInterfaceArrayFloat3, &NiagaraDataInterfaceCurve]
 usedDynamicInputs [
  &SelectFloatFromArray
  &TransformVector
  &SetBoolByFloatComparison
 ]
 usedModules [
  &SystemState
  &self:SystemSpawnScript.SystemScriptSource.SystemScriptGraph.NiagaraNodeAssignment_2.SetVariables_#1
  &self:SystemSpawnScript.SystemScriptSource.SystemScriptGraph.NiagaraNodeAssignment_1.SetVariables_#2
  &FN_Integrator
  &self:SystemSpawnScript.SystemScriptSource.SystemScriptGraph.NiagaraNodeAssignment_0.SetVariables_#3
 ]
}
emitters {
 Decal_Particle {
  emitterUpdateScript [
   &EmitterState(
    Inactive Response=::NewEnumerator1
    Life Cycle Mode=::NewEnumerator1
    Loop Behavior=::NewEnumerator1
    Loop Duration Mode=::NewEnumerator1
    UseLoopDelay=false
   )
   &SpawnBurst_Instantaneous(Spawn Count=5, Use Spawn Probability=false)
  ]
  particleSpawnScript [
   &FN_SetSpawnGroup(Number of Spawn Groups=5)
   /Niagara/Modules/Spawn/Initialization/InitializeParticle(
    Module.Write Color=true
    Module.Write Lifetime=true
    Module.Write Mass=true
    Module.Write Material Random=true
    Module.Write Position=true
    Module.Write Scale=false
    Module.Write SpriteRotation=false
    Module.Write SpriteSize=true
    Position=&SimulationPosition()
    Sprite Size=&Vector2DFromFloat(Value=35)
   )
   &CameraOffset(Camera Offset Amount=20)
   &DynamicMaterialParameters(
    Index 0 Param 1=10
    Module.Param0WriteEnabled=true
    Module.Param1WriteEnabled=false
    Module.Param2WriteEnabled=false
    Module.Param3WriteEnabled=false
   ) {module=DynamicMaterialParameters001}
  ]
  particleUpdateScript [
   /Game/Effects/Niagara/Modules/Skydiving/FN_DivingAttributes/v5/FN_DivingAttributes(Left Calf Index=10, SkeletalMesh:NiagaraDataInterfaceSkeletalMesh=$User.SkeletalMesh)
   &ParticleState(Kill Particles When Lifetime Has Elapsed=false, Let Infinitely Lived Particles Die When Emitter Deactivates=true)
   &SolveForcesAndVelocity(Clamp Velocity=false, Limit Acceleration=false)
   &SkeletalMeshLocation(
    Bone / Socket Index=$Particles.SpawnGroup
    Bone Sampling Mode=::NewEnumerator4
    Inherit Velocity=false
    MaskSpawnGroup=false
    Skeletal Mesh=$User.SkeletalMesh
   )
   &DynamicMaterialParameters(
    Index 0 Param 1=2
    Module.Param0WriteEnabled=true
    Module.Param1WriteEnabled=true
    Module.Param2WriteEnabled=true
    Module.Param3WriteEnabled=true
   ) {enabled=false}
   &ScaleColor(
    Scale Alpha=&Multiply_Float(
     A=&FNTimeOfDayScalar(Night Scale=0.2, Morning Scale=0.4, Day Scale=0.6, Evening Scale=0.4)
     B=0.5
    )
    ScaleA=true
    ScaleRGB=true
   )
   &Color(Color=&LerpLinearColors(
     StartColor=<rgba 0 2 3 0.33>
     EndColor=<rgba 1 2 3 0.33>
     LerpFactor=&NormalizeFloat(
      Value=&Sine(
       Normalized Angle=&Multiply_Float(A=$Engine.System.Age, B=&Lerp_Float(A=1, B=3, Alpha=$Particles.Dive))
       Period=3
       Scale=1
       Bias=0
      )
      Min=-1
      Max=1
     )
    ))
  ]
  rendererClasses [&NiagaraSpriteRendererProperties]
  renderers [{rendererClass=&NiagaraSpriteRendererProperties}]
 }
 Flares_Hand {
  emitterUpdateScript [
   &EmitterState(Life Cycle Mode=::NewEnumerator1, Loop Duration=1, UseLoopDelay=false)
   &SpawnRate(SpawnRate=15, Use Spawn Probability=true)
  ]
  particleSpawnScript [
   /Game/Effects/Niagara/Modules/Skydiving/FN_DivingAttributes/FN_DivingAttributes_v2(Accumulate Dive=0.1, Accumulate Rest=-0.1, SkeletalMesh:NiagaraDataInterfaceSkeletalMesh=$User.SkeletalMesh) {enabled=false}
   &FN_SetSpawnGroup(Number of Spawn Groups=2)
   &InitializeParticle(
    Color=<rgba 0.5 4 5 1>
    Lifetime=0.6
    Module.Write Color=true
    Module.Write Lifetime=true
    Module.Write Mass=false
    Module.Write Material Random=true
    Module.Write Position=false
    Module.Write Scale=false
    Module.Write SpriteRotation=true
    Module.Write SpriteSize=true
    Sprite Rotation=&RandomRangeFloat(Minimum=0, Maximum=360)
    Sprite Size=&Vector2DFromFloat(Value=&RandomRangeFloat(Minimum=100, Maximum=100))
   )
  ]
  particleUpdateScript [
   &ParticleState()
   &SkeletalMeshLocation(
    Bone / Socket Index=$Particles.SpawnGroup
    Bone Sampling Mode=::NewEnumerator4
    Inherit Velocity=false
    MaskSpawnGroup=false
    Sampled Position Coordinate Space=::Local
    Skeletal Mesh=$User.SkeletalMesh
   ) {enabled=false, module=SkeletalMeshLocation001}
   &SkeletalMeshLocation(
    Bone / Socket Index=$Particles.SpawnGroup
    Bone Sampling Mode=::NewEnumerator4
    Inherit Velocity=false
    Interpolate Mesh Sampling=false
    MaskSpawnGroup=false
    Skeletal Mesh=$User.SkeletalMesh
   )
   /Game/Effects/Niagara/Modules/Skydiving/FN_SkydivePositionOffset_v1/FN_SkydivePositionOffset(
    Bone Orientation:Quat4f=$Output.SkeletalMeshLocation.SampledRotation
    Bone Position:Vector3f=$Output.SkeletalMeshLocation.SampledPosition
    Offset:NiagaraFloat=/Game/Effects/Niagara/DynamicInputs/Skydiving/SkydiveHandsFeetFloatModifier(Hands:NiagaraFloat=20, Feet:NiagaraFloat=45)
   ) {enabled=false}
   /Game/Effects/Niagara/Modules/Skydiving/FN_Skydive_Velocity_v4/FN_Skydive_Velocity(
    Bone Orientation:Quat4f=$Output.SkeletalMeshLocation.SampledRotation
    Bone Position:Vector3f=$Output.SkeletalMeshLocation.SampledPosition
    Dive:NiagaraFloat=$System.Dive
    FE Speed=1
    FE Spread=0.1
    SpawnGroup:NiagaraInt32=$Particles.SpawnGroup
    Write to Intrinsic Properties=false
   ) {enabled=false}
   &ScaleSpriteSize(Scale Factor=&Vector2DFromFloat(Value=&FloatFromCurve(FloatCurve=DataInterface() {propertyValues {Curve {keys [
          {arriveTangent=1.5832, interpMode=RCIM_Cubic, time=0, value=0.5}
          {arriveTangent=0.5893, interpMode=RCIM_Cubic, leaveTangent=0.5893, tangentMode=RCTM_User, time=1, value=1}
         ]}}})))
   &SolveForcesAndVelocity(Clamp Velocity=false, Limit Acceleration=false)
   &SpriteRotationRate(Rotation Rate=&MakeCustomFloatFromBool(Bool=&RandomBool(), True Float=10, False Float=-10)) {enabled=false}
   &ScaleColor(
    Scale Alpha=&FloatFromCurve(
     FloatCurve=DataInterface() {propertyValues {Curve {keys [
         {arriveTangent=4.9245, interpMode=RCIM_Cubic, time=0, value=0}
         <at 0.33 1 RCIM_Cubic>
         {interpMode=RCIM_Cubic, leaveTangent=-1.9375, time=1, value=0}
        ]}}}
     Scale Curve=0.25
    )
    Scale RGB=&VectorFromFloat(Value=&FloatFromCurve(
      FloatCurve=DataInterface() {propertyValues {Curve {keys [
          {arriveTangent=8517.9209, interpMode=RCIM_Cubic, leaveTangent=2.9045, tangentMode=RCTM_User, time=0, value=0.1}
          <at 0.33 1 RCIM_Cubic>
          {interpMode=RCIM_Cubic, leaveTangent=-4.1243, time=1, value=0.1}
         ]}}}
      Scale Curve=&Multiply_Float(A=&FNTimeOfDayScalar(Night Scale=0.25, Morning Scale=0.5, Evening Scale=0.5), B=1)
     ))
    ScaleA=true
    ScaleRGB=true
   )
   &CameraOffset(Camera Offset Amount=20)
   &DynamicMaterialParameters(
    Index 0 Param 1=0.6
    Index 0 Param 2=1
    Module.Param0WriteEnabled=true
    Module.Param1WriteEnabled=true
    Module.Param2WriteEnabled=true
    Module.Param3WriteEnabled=true
   ) {enabled=false}
  ]
  rendererClasses [&NiagaraSpriteRendererProperties]
  renderers [{rendererClass=&NiagaraSpriteRendererProperties}]
 }
 Light {
  emitterUpdateScript [
   &EmitterState(
    Inactive Response=::NewEnumerator1
    Life Cycle Mode=::NewEnumerator1
    Loop Behavior=::NewEnumerator1
    Loop Duration Mode=::NewEnumerator1
    UseLoopDelay=false
   )
   &SpawnBurst_Instantaneous(Spawn Count=2, Use Spawn Probability=false)
  ]
  particleSpawnScript [
   &FN_SetSpawnGroup(Number of Spawn Groups=2)
   /Niagara/Modules/Spawn/Initialization/V2/InitializeParticle(Color=<rgba 2.5 20 25 1>, UsePositionOffset=false, Write Material Random=true)
  ]
  particleUpdateScript [
   &FN_DivingAttributes_v2(Accumulate Dive=0.1, Accumulate Rest=-0.1, Bone Index=0, SkeletalMesh:NiagaraDataInterfaceSkeletalMesh=$User.Skeletal Mesh01)
   /SetVariables(Particles.Dive01:NiagaraFloat=&Multiply_Float(A=1, B=$Particles.Dive)) {isSetParameters=true}
   &ParticleState(Kill Particles When Lifetime Has Elapsed=false, Let Infinitely Lived Particles Die When Emitter Deactivates=true)
   &SkeletalMeshLocation(
    Apply Sampled Orientation to Mesh Particles=true
    Bone / Socket Index=$Particles.SpawnGroup
    Bone Sampling Mode=::NewEnumerator4
    Inherit Velocity=false
    MaskSpawnGroup=false
    Sampled Position Offset Coordinate Space:ENiagaraMeshTransforms=::NewEnumerator2
    Skeletal Mesh=$User.SkeletalMesh
   )
   &ScaleColor(
    Scale Alpha=&FNTimeOfDayScalar(Night Scale=0.5, Morning Scale=0.7, Evening Scale=0.7)
    Scale RGB=&VectorFromFloat(Value=&Lerp_Float(A=0.5, B=1, Alpha=$Particles.Dive01))
    ScaleA=true
    ScaleRGB=true
   ) {module=ScaleColor001}
  ]
  rendererClasses [&NiagaraLightRendererProperties]
  renderers [{rendererClass=&NiagaraLightRendererProperties, rendererIndex=0}]
 }
 Mesh_Trails {
  emitterUpdateScript [
   &EmitterState(
    Inactive Response=::NewEnumerator1
    Life Cycle Mode=::NewEnumerator1
    Loop Behavior=::NewEnumerator1
    Loop Duration Mode=::NewEnumerator1
    UseLoopDelay=false
   )
   &SpawnBurst_Instantaneous(Spawn Count=5, Use Spawn Probability=false)
  ]
  particleSpawnScript [
   &FN_SetSpawnGroup(Number of Spawn Groups=5)
   &InitializeParticle(
    Color=<rgba 1 4 5 1>
    Mesh Scale=&SelectVectorFromArray(
     Direct Array Index=$Particles.SpawnGroup
     Vector Selection Array=DataInterface() {propertyValues {FloatData [
        <0.2 0.2 0.45>
        <0.2 0.2 0.45>
        <0.2 0.2 0.4>
        <0.2 0.2 0.4>
        <0.3 0.3 0.4>
        <0.4 0.4 0.4>
        <0.4 0.4 0.4>
        <0.4 0.4 0.4>
        <0.4 0.4 0.5>
        <0.5 0.5 0.5>
       ]}}
    )
    Mesh Scale Mode=::NewEnumerator4
    UsePositionOffset=false
    Write Material Random=true
   )
   &DynamicMaterialParameters(
    Index 0 Param 2=&RandomRangeFloat(Minimum=0, Maximum=10)
    Index 0 Param 3=30
    Index 0 Param 4=1
    Module.Param0WriteEnabled=true
    Module.Param1WriteEnabled=true
    Module.Param2WriteEnabled=false
    Module.Param3WriteEnabled=true
   )
   /SetVariables(Particles.RotatorSpeedRandom:NiagaraFloat=&RandomRangeFloat(Minimum=0.3, Maximum=-0.3)) {isSetParameters=true}
   /SetVariables(Particles.WPO=25) {isSetParameters=true}
   &CameraOffset(Camera Offset Amount=10)
   &InitialMeshOrientation(
    Orientation Axis=<0 0 1>
    Orientation Coordinate Space=::Local
    Orientation Vector=<0 0 1>
    Rotation=&MakeVector(X=0, Y=0, Z=&RandomRangeFloat(Minimum=0, Maximum=360))
    Rotation Coordinate Space=::NewEnumerator0
    Use Orientation Vector=false
    Use Rotation Vector=true
   )
  ]
  particleUpdateScript [
   &ParticleState(Kill Particles When Lifetime Has Elapsed=false, Let Infinitely Lived Particles Die When Emitter Deactivates=true)
   /SetVariables(Particles.MeshOrientation=&MultiplyQuaternion(
     Quaternion A=&FindQuatBetweenTwoVectors(From Vector=<0 0 -1>, To Vector=$System.SkydiveVelocity)
     Quaternion B=$Particles.Initial.MeshOrientation
    )) {isSetParameters=true}
   &InterpolateOverTime(Rate Of Change=15, Target Vector Value:NiagaraPosition=$Particles.Position, Variable Type=::NewEnumerator2)
   /SetVariables(Particles.InterpolateDynamic:Vector4f=&MakeFloat4FromVectorAndFloat("Vector (RGB)"=$Output.InterpolateOverTime.Moving Average, "Float (Alpha)"=1)) {isSetParameters=true}
   &SkeletalMeshLocation(
    Bone / Socket Index=$Particles.SpawnGroup
    Bone Sampling Mode=::NewEnumerator4
    Inherit Velocity=false
    Interpolate Mesh Sampling=false
    MaskSpawnGroup=false
    Skeletal Mesh=$User.SkeletalMesh
   )
   &FN_SkydivePositionOffset(
    Bone Orientation:Quat4f=$Output.SkeletalMeshLocation.SampledRotation
    Bone Position:Vector3f=$Output.SkeletalMeshLocation.SampledPosition
    Offset:NiagaraFloat=&SkydiveHandsFeetFloatModifier(Hands:NiagaraFloat=20, Feet:NiagaraFloat=45)
   ) {enabled=false}
   &FN_Skydive_Velocity(
    Bone Orientation:Quat4f=$Output.SkeletalMeshLocation.SampledRotation
    Bone Position:Vector3f=$Output.SkeletalMeshLocation.SampledPosition
    Dive:NiagaraFloat=$System.Dive
    FE Speed=1
    SpawnGroup:NiagaraInt32=$Particles.SpawnGroup
    Write to Intrinsic Properties=false
   ) {enabled=false}
   /SetVariables(Particles.MeshOrientation=&FindQuatBetweenTwoVectors(From Vector=<0 0 1>, To Vector=$Output.FN_Skydive_Velocity.Velocity)) {enabled=false, isSetParameters=true}
   &ScaleMeshSize(Scale Factor=&Lerp_Vector(A=<1 1 0.5>, B=<0.85 0.85 0.75>, Alpha=$System.Dive))
   &ScaleMeshSize(Scale Factor=&FN_FrontEndVector3Modifier(Gameplay=<1 1 1>, Front End=<1.25 1.25 2>)) {module=ScaleMeshSize001}
   &SolveForcesAndVelocity(Clamp Velocity=false, Limit Acceleration=false)
   &DynamicMaterialParameters(
    Index 0 Param 2=&RandomRangeFloat(Minimum=0, Maximum=5)
    Index 0 Param 3=&Multiply_Float(A=$Particles.WPO, B=&Lerp_Float(A=0.2, B=0.4, Alpha=$System.Dive))
    Index 2 Param 1=&Lerp_Float(A=0.2, B=1, Alpha=&Ceil_Float(Float=$System.Dive))
    Index 2 Param 2=&Lerp_Float(A=0, B=0.3, Alpha=&Ceil_Float(Float=$System.Dive))
    Module.Param0WriteEnabled=false
    Module.Param0WriteEnabled2=true
    Module.Param1WriteEnabled=true
    Module.Param1WriteEnabled2=true
    Module.Param2WriteEnabled=true
    Module.Param2WriteEnabled2=false
    Module.Param3WriteEnabled=false
    Module.Param3WriteEnabled2=false
    Write Parameter Index 2=true
   ) {module=DynamicMaterialParameters001}
  ]
  rendererClasses [&NiagaraMeshRendererProperties]
  renderers [{rendererClass=&NiagaraMeshRendererProperties, rendererIndex=0}]
 }
 Sparkles_Stream {
  simTarget=GPUComputeSim
  emitterUpdateScript [
   &EmitterState(Life Cycle Mode=::NewEnumerator1, UseLoopDelay=false)
   &SpawnRate(Spawn Probability=0.33, SpawnRate=200, Use Spawn Probability=true)
  ]
  particleSpawnScript [
   &FN_SetSpawnGroup(Number of Spawn Groups=5)
   /Niagara/Modules/Spawn/Initialization/InitializeParticle(
    Color=<rgba 0.5 4 5 1>
    Lifetime=&RandomRangeFloat(Minimum=1.2, Maximum=1.6)
    Module.Write Color=true
    Module.Write Lifetime=true
    Module.Write Mass=false
    Module.Write Material Random=true
    Module.Write Position=false
    Module.Write Scale=false
    Module.Write SpriteRotation=true
    Module.Write SpriteSize=true
    Sprite Rotation=&RandomRangeFloat(Minimum=0, Maximum=360)
    Sprite Size=&Vector2DFromFloat(Value=&RandomRangeFloat(Minimum=10, Maximum=15))
   )
   &AddVelocity(Velocity=&FN_FrontEndVector3Modifier(Gameplay=<0 0 0>, Front End=<0 0 0>)) {enabled=false}
  ]
  particleUpdateScript [
   &FN_DivingAttributes(Left Calf Index=10, SkeletalMesh:NiagaraDataInterfaceSkeletalMesh=$User.SkeletalMesh) {enabled=false}
   &ParticleState()
   &SkeletalMeshLocation(
    Bone / Socket Index=$Particles.SpawnGroup
    Bone Sampling Mode=::NewEnumerator4
    Inherit Velocity=false
    MaskSpawnGroup=false
    Sampled Position Offset=&Multiply_VectorByFloat(
     Vector=&NormalizeVector(Vector To Normalize=$System.SkydiveVelocity)
     Float=&FloatFromCurve(
      FloatCurve=DataInterface() {propertyValues {Curve {keys [<at 0 1>,<at 1 0>]}}}
      Scale Curve=&Lerp_Float(A=-150, B=-200, Alpha=$System.Dive)
     )
    )
    Skeletal Mesh=$User.SkeletalMesh
   )
   &FN_SkydivePositionOffset(
    Bone Orientation:Quat4f=$Output.SkeletalMeshLocation.SampledRotation
    Bone Position:Vector3f=$Output.SkeletalMeshLocation.SampledPosition
    Offset:NiagaraFloat=&SkydiveHandsFeetFloatModifier(Hands:NiagaraFloat=20, Feet:NiagaraFloat=45)
   ) {enabled=false}
   &FN_Skydive_Velocity(
    Bone Orientation:Quat4f=$Output.SkeletalMeshLocation.SampledRotation
    Bone Position:Vector3f=$Output.SkeletalMeshLocation.SampledPosition
    Dive:NiagaraFloat=$System.Dive
    FE Speed=100
    SpawnGroup:NiagaraInt32=$Particles.SpawnGroup
    Write to Intrinsic Properties=false
   ) {enabled=false}
   &SpriteRotationRate(Rotation Rate=&MakeCustomFloatFromBool(Bool=&RandomBool(), True Float=220, False Float=-220))
   &SolveForcesAndVelocity(Clamp Velocity=false, Limit Acceleration=false)
   &ScaleColor(
    Scale Alpha=&FloatFromCurve(
     FloatCurve=DataInterface() {propertyValues {Curve {keys [<at 0 0>,<at 0.5 1>,<at 0.75 1>,<at 1 0>]}}}
     Scale Curve=&RandomRangeFloat(Minimum=0.05, Maximum=0.3)
    )
    Scale RGB=&VectorFromFloat(Value=&FloatFromCurve(
      FloatCurve=DataInterface() {propertyValues {Curve {keys [<at 0 0>,<at 0.5 1>,<at 0.75 1>,<at 1 0>]}}}
      Scale Curve=&Multiply_Float(
       A=&FNTimeOfDayScalar(Night Scale=0.25, Morning Scale=0.5, Evening Scale=0.5)
       B=&RandomRangeFloat(Minimum=0.05, Maximum=0.3)
      )
     ))
    ScaleA=true
    ScaleRGB=true
   )
   &DynamicMaterialParameters(
    Index 0 Param 1=0.25
    Index 0 Param 2=3
    Module.Param0WriteEnabled=true
    Module.Param1WriteEnabled=true
    Module.Param2WriteEnabled=true
    Module.Param3WriteEnabled=true
   )
  ]
  rendererClasses [&NiagaraSpriteRendererProperties]
  renderers [{rendererClass=&NiagaraSpriteRendererProperties}]
 }
 Sparkles_Surround {
  emitterUpdateScript [
   &EmitterState(Life Cycle Mode=::NewEnumerator1, UseLoopDelay=false)
   &SpawnRate(Spawn Probability=0.33, SpawnRate=200, Use Spawn Probability=true)
  ]
  particleSpawnScript [
   &FN_SetSpawnGroup(Number of Spawn Groups=5) {enabled=false}
   &InitializeParticle(
    Color=<rgba 0.5 4 5 1>
    Lifetime=&RandomRangeFloat(Minimum=1.2, Maximum=1.6)
    Module.Write Color=true
    Module.Write Lifetime=true
    Module.Write Mass=false
    Module.Write Material Random=true
    Module.Write Position=false
    Module.Write Scale=false
    Module.Write SpriteRotation=true
    Module.Write SpriteSize=true
    Sprite Rotation=&RandomRangeFloat(Minimum=0, Maximum=360)
    Sprite Size=&Vector2DFromFloat(Value=&RandomRangeFloat(Minimum=10, Maximum=15))
   )
   &AddVelocity(Velocity=&FN_FrontEndVector3Modifier(Gameplay=<0 0 0>, Front End=<0 0 0>)) {enabled=false}
   &SphereLocation(
    MaskSpawnGroup=false
    Override Seed=false
    Sphere Coordinate Space=::Simulation
    Sphere Radius=200
    Surface Only=false
    UseNonUniformScale=false
    UseOffset=false
    UseSphereOrientation=false
   )
  ]
  particleUpdateScript [
   &FN_DivingAttributes(Left Calf Index=10, SkeletalMesh:NiagaraDataInterfaceSkeletalMesh=$User.SkeletalMesh) {enabled=false}
   &ParticleState()
   &SkeletalMeshLocation(
    Bone / Socket Index=4
    Bone Sampling Mode=::NewEnumerator4
    Inherit Velocity=false
    MaskSpawnGroup=false
    Sampled Position Offset=&Add_Vector(
     A=Linked() {linkedVariable {name=Particles.Initial.Position, type=NiagaraPosition}}
     B=&Multiply_VectorByFloat(
      Vector=&Multiply_VectorByFloat(
       Vector=&NormalizeVector(Vector To Normalize=&Multiply_Vector(A=$System.SkydiveVelocity, B=<0.5 0.5 1>))
       Float=&FN_FrontEndFloatModifier(Gameplay=1, Front End=0.5)
      )
      Float=&FloatFromCurve(
       FloatCurve=DataInterface() {propertyValues {Curve {keys [<at 0 1>,<at 1 -1>]}}}
       Scale Curve=&Lerp_Float(A=-150, B=-200, Alpha=$System.Dive)
      )
     )
    )
    Skeletal Mesh=$User.SkeletalMesh
   )
   &FN_SkydivePositionOffset(
    Bone Orientation:Quat4f=$Output.SkeletalMeshLocation.SampledRotation
    Bone Position:Vector3f=$Output.SkeletalMeshLocation.SampledPosition
    Offset:NiagaraFloat=&SkydiveHandsFeetFloatModifier(Hands:NiagaraFloat=20, Feet:NiagaraFloat=45)
   ) {enabled=false}
   &FN_Skydive_Velocity(
    Bone Orientation:Quat4f=$Output.SkeletalMeshLocation.SampledRotation
    Bone Position:Vector3f=$Output.SkeletalMeshLocation.SampledPosition
    Dive:NiagaraFloat=$System.Dive
    FE Speed=100
    SpawnGroup:NiagaraInt32=$Particles.SpawnGroup
    Write to Intrinsic Properties=false
   ) {enabled=false}
   &SpriteRotationRate(Rotation Rate=&MakeCustomFloatFromBool(Bool=&RandomBool(), True Float=220, False Float=-220))
   &SolveForcesAndVelocity(Clamp Velocity=false, Limit Acceleration=false)
   &ScaleColor(
    Scale Alpha=&FloatFromCurve(
     FloatCurve=DataInterface() {propertyValues {Curve {keys [<at 0 0>,<at 0.33 1>,<at 0.66 1>,<at 1 0>]}}}
     Scale Curve=&RandomRangeFloat(Minimum=0.05, Maximum=0.3)
    )
    Scale RGB=&VectorFromFloat(Value=&FloatFromCurve(
      FloatCurve=DataInterface() {propertyValues {Curve {keys [<at 0 0>,<at 0.33 1>,<at 0.66 1>,<at 1 0>]}}}
      Scale Curve=&Multiply_Float(
       A=&FNTimeOfDayScalar(Night Scale=0.25, Morning Scale=0.5, Evening Scale=0.5)
       B=&RandomRangeFloat(Minimum=0.1, Maximum=0.6)
      )
     ))
    ScaleA=true
    ScaleRGB=true
   )
   &DynamicMaterialParameters(
    Index 0 Param 1=0.25
    Index 0 Param 2=3
    Module.Param0WriteEnabled=true
    Module.Param1WriteEnabled=true
    Module.Param2WriteEnabled=true
    Module.Param3WriteEnabled=true
   )
  ]
  rendererClasses [&NiagaraSpriteRendererProperties]
  renderers [{rendererClass=&NiagaraSpriteRendererProperties}]
 }
}
summary {
 systemName=NS_Contrail_WavesOfLight
 emitters [
  {emitterName=Mesh_Trails, rendererClasses [&NiagaraMeshRendererProperties]}
  {emitterName=Decal_Particle, rendererClasses [&NiagaraSpriteRendererProperties]}
  {emitterName=Flares_Hand, rendererClasses [&NiagaraSpriteRendererProperties]}
  {emitterName=Sparkles_Stream, simTarget=GPUComputeSim, rendererClasses [&NiagaraSpriteRendererProperties]}
  {emitterName=Sparkles_Surround, rendererClasses [&NiagaraSpriteRendererProperties]}
  {emitterName=Light, rendererClasses [&NiagaraLightRendererProperties]}
 ]
 userVariables [
  User.bIsFrontEnd {type=NiagaraBool}
  User.bIsFrontEndPreview {type=NiagaraBool, defaultValue=true}
  User.ParaGlideLeanAlpha {type=NiagaraFloat, defaultValue=0, description=""}
  User.Skeletal Mesh01 {
   type=NiagaraDataInterfaceSkeletalMesh
   defaultValue {
    struct=NiagaraExt_VariableValue_DataInterface
    value {dataInterface=&self:NiagaraDataInterfaceSkeletalMesh_1}
   }
  }
  User.Skeletal Mesh02 {
   type=NiagaraDataInterfaceSkeletalMesh
   defaultValue {
    struct=NiagaraExt_VariableValue_DataInterface
    value {dataInterface=&self:NiagaraDataInterfaceSkeletalMesh_2}
   }
  }
  User.SkeletalMesh {
   type=NiagaraDataInterfaceSkeletalMesh
   defaultValue {
    struct=NiagaraExt_VariableValue_DataInterface
    value {dataInterface=&self:NiagaraDataInterfaceSkeletalMesh_0}
   }
  }
 ]
}
